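# Idle body effect for the PawJasmineSword character cosmetic, emitting glitch sprites around leg, shoulder, chest and foot sockets.
# One emitter runs seven spawn groups whose index selects socket and curve-shaped random offset; bisFootHidden stops foot spawning.
# Curl noise jitters particles, sprite size fades near the camera, and five renderers layer sprites with alternate sizes and colors.
# System ElimAlpha raises lifetime, size, brightness and an extra camera-offset layer's intensity.
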NiagaraSystem /BRCosmetics/Characters/Player/Female/Medium/Bodies/F_MED_PawJasmineSword/FX/NS_Character_PawJasmineSword_BodyIdle
dependencies {
 usedDataInterfaces [
  &NiagaraDataInterfaceColorCurve
  &NiagaraDataInterfaceCurve
  &NiagaraDataInterfaceCamera
  &NiagaraDataInterfaceArrayFloat3
  &NiagaraDataInterfaceArrayFloat
  &NiagaraDataInterfaceSocketReader
 ]
 usedDynamicInputs [
  &SetBoolByFloatComparison
  &Boolean_AndOperation
  &SetBoolByIntComparison
 ]
 usedModules [
  &self:SystemSpawnScript.SystemScriptSource.SystemScriptGraph.NiagaraNodeAssignment_0.SetVariables_#1
  &self:SystemSpawnScript.SystemScriptSource.SystemScriptGraph.NiagaraNodeAssignment_3.SetVariables_#2
  &SystemState
  /Game/Effects/Niagara/Modules/Logic/UpdateValueStateChange
  &Timeline
  &self:SystemSpawnScript.SystemScriptSource.SystemScriptGraph.NiagaraNodeAssignment_4.SetVariables_#3
  &self:SystemSpawnScript.SystemScriptSource.SystemScriptGraph.NiagaraNodeAssignment_2.SetVariables_#4
  &self:SystemSpawnScript.SystemScriptSource.SystemScriptGraph.NiagaraNodeAssignment_1.SetVariables_#5
 ]
}
emitters {NE_GlitchSprites {
  emitterUpdateScript [
   &EmitterState()
   &SpawnRate(SpawnRate=&Multiply_Float(A=20, B=$System.SpawnMultiplier), Use Spawn Probability=false)
   &SpawnRate(SpawnGroup=1, SpawnRate=&Multiply_Float(A=20, B=$System.SpawnMultiplier), Use Spawn Probability=false) {module=SpawnRate001}
   &SpawnRate(SpawnGroup=2, SpawnRate=&Multiply_Float(A=15, B=$System.SpawnMultiplier), Use Spawn Probability=false) {module=SpawnRate002}
   &SpawnRate(SpawnGroup=3, SpawnRate=&Multiply_Float(A=15, B=$System.SpawnMultiplier), Use Spawn Probability=false) {module=SpawnRate003}
   &SpawnRate(SpawnGroup=4, SpawnRate=&Multiply_Float(A=10, B=$System.SpawnMultiplier), Use Spawn Probability=false) {module=SpawnRate004}
   &SpawnRate(
    SpawnGroup=5
    SpawnRate=&Multiply_Float(
     A=&MakeCustomFloatFromBool(Bool=$User.bisFootHidden, True Float=0, False Float=10)
     B=$System.SpawnMultiplier
    )
    Use Spawn Probability=false
   ) {module=SpawnRate005}
   &SpawnRate(
    SpawnGroup=6
    SpawnRate=&Multiply_Float(
     A=&MakeCustomFloatFromBool(Bool=$User.bisFootHidden, True Float=0, False Float=10)
     B=$System.SpawnMultiplier
    )
    Use Spawn Probability=false
   ) {module=SpawnRate006}
  ]
  particleSpawnScript [
   &InitializeParticle(
    Color=&ColorFromCurve(
     ColorCurve=DataInterface() {propertyValues {
       AlphaCurve {keys [<at 0 1 RCIM_Cubic>]}
       BlueCurve {keys [<at 0.4616 0 RCIM_Cubic>,<at 0.4893 0 RCIM_Cubic>]}
       GreenCurve {keys [<at 0.4616 0.2438 RCIM_Cubic>,<at 0.4893 0.1091 RCIM_Cubic>]}
       RedCurve {keys [<at 0.4616 2 RCIM_Cubic>,<at 0.4893 1.5 RCIM_Cubic>]}
      }}
     CurveIndex=&RandomRangeFloat(Minimum=0, Maximum=1)
    )
    Lifetime=&Lerp_Float(
     A=&RandomRangeFloat(Minimum=0.3, Maximum=0.5)
     B=&RandomRangeFloat(Minimum=0.5, Maximum=0.7)
     Alpha=$System.ElimAlpha
    )
    Mesh Scale Mode=::NewEnumerator3
    Mesh Uniform Scale=0.1
    Sprite Size=&Multiply_Vector2D_ByFloat(Float=&RandomRangeFloat(Minimum=5, Maximum=6), Vector2D=<0.75 2.5>)
    Sprite Size Mode=::NewEnumerator4
    UsePositionOffset=true
    Write Material Random=true
   ) {module=InitializeParticle001}
   /SetVariables(Particles.IndexType:NiagaraInt32=&SpawnGroup()) {isSetParameters=true}
   /SetVariables(Particles.RandomFloat=&RandomRangeFloat(Minimum=0, Maximum=1)) {isSetParameters=true}
   &DynamicMaterialParameters(
    Index 0 Param 4=&RandomRangeFloat(Minimum=-1, Maximum=0.1)
    Index 1 Param 1=&RandomRangeFloat(Minimum=-100, Maximum=100)
    Module.Param0WriteEnabled=true
    Module.Param0WriteEnabled1=true
    Module.Param1WriteEnabled=true
    Module.Param1WriteEnabled1=true
    Module.Param2WriteEnabled=true
    Module.Param2WriteEnabled1=true
    Module.Param3WriteEnabled=true
    Module.Param3WriteEnabled1=true
    Write Parameter Index 1=true
   ) {module=DynamicMaterialParameters001}
   /SetVariables(Particles.LegOffset:Vector3f=&MakeVector(
     X=&Lerp_Float(
      A=-35
      B=-10
      Alpha=&FloatFromCurve(
       FloatCurve=DataInterface() {propertyValues {Curve {keys [{interpMode=RCIM_Cubic, tangentMode=RCTM_User, time=0, value=0}, {interpMode=RCIM_Cubic, tangentMode=RCTM_User, time=1, value=1}]}}}
       CurveIndex=$Particles.RandomFloat
      )
     )
     Y=&Lerp_Float(
      A=12
      B=-20
      Alpha=&FloatFromCurve(
       FloatCurve=DataInterface() {propertyValues {Curve {keys [{interpMode=RCIM_Cubic, tangentMode=RCTM_User, time=0, value=0}, {interpMode=RCIM_Cubic, tangentMode=RCTM_User, time=1, value=1}]}}}
       CurveIndex=$Particles.RandomFloat
      )
     )
     Z=&Lerp_Float(
      A=0
      B=-15
      Alpha=&FloatFromCurve(
       FloatCurve=DataInterface() {propertyValues {Curve {keys [
           {interpMode=RCIM_Cubic, tangentMode=RCTM_User, time=0, value=0}
           {interpMode=RCIM_Cubic, tangentMode=RCTM_User, time=0.5, value=1}
           {interpMode=RCIM_Cubic, tangentMode=RCTM_User, time=1, value=0}
          ]}}}
       CurveIndex=$Particles.RandomFloat
      )
     )
    )) {isSetParameters=true}
   /SetVariables(Particles.ShoulderOffset:Vector3f=&MakeVector(
     X=&Lerp_Float(
      A=0
      B=0
      Alpha=&FloatFromCurve(
       FloatCurve=DataInterface() {propertyValues {Curve {keys [
           {arriveTangent=-0.3783, interpMode=RCIM_Cubic, leaveTangent=-0.3783, tangentMode=RCTM_User, time=0, value=0}
           {interpMode=RCIM_Cubic, tangentMode=RCTM_User, time=0.5, value=1}
           {arriveTangent=0.3077, interpMode=RCIM_Cubic, leaveTangent=0.3077, tangentMode=RCTM_User, time=1, value=0}
          ]}}}
       CurveIndex=$Particles.RandomFloat
      )
     )
     Y=&Lerp_Float(
      A=0
      B=15
      Alpha=&FloatFromCurve(
       FloatCurve=DataInterface() {propertyValues {Curve {keys [
           {arriveTangent=-0.3783, interpMode=RCIM_Cubic, leaveTangent=-0.3783, tangentMode=RCTM_User, time=0, value=1}
           {interpMode=RCIM_Cubic, tangentMode=RCTM_User, time=0.5, value=0}
           {arriveTangent=0.3077, interpMode=RCIM_Cubic, leaveTangent=0.3077, tangentMode=RCTM_User, time=1, value=-1}
          ]}}}
       CurveIndex=$Particles.RandomFloat
      )
     )
     Z=&Lerp_Float(
      A=0
      B=15
      Alpha=&FloatFromCurve(
       FloatCurve=DataInterface() {propertyValues {Curve {keys [
           {arriveTangent=-0.3783, interpMode=RCIM_Cubic, leaveTangent=-0.3783, tangentMode=RCTM_User, time=0, value=0}
           {interpMode=RCIM_Cubic, tangentMode=RCTM_User, time=0.5, value=1}
           {arriveTangent=0.3077, interpMode=RCIM_Cubic, leaveTangent=0.3077, tangentMode=RCTM_User, time=1, value=0}
          ]}}}
       CurveIndex=$Particles.RandomFloat
      )
     )
    )) {isSetParameters=true}
   /SetVariables(Particles.ChestOffset:Vector3f=&MakeVector(
     X=&Lerp_Float(
      A=10
      B=20
      Alpha=&FloatFromCurve(
       FloatCurve=DataInterface() {propertyValues {Curve {keys [
           {arriveTangent=-0.3783, interpMode=RCIM_Cubic, leaveTangent=-0.3783, tangentMode=RCTM_User, time=0, value=0.7}
           {interpMode=RCIM_Cubic, tangentMode=RCTM_User, time=0.5, value=0}
           {arriveTangent=0.3077, interpMode=RCIM_Cubic, leaveTangent=0.3077, tangentMode=RCTM_User, time=1, value=0.7}
          ]}}}
       CurveIndex=$Particles.RandomFloat
      )
     )
     Y=&Lerp_Float(
      A=18
      B=10
      Alpha=&FloatFromCurve(
       FloatCurve=DataInterface() {propertyValues {Curve {keys [
           {arriveTangent=-3.1485, interpMode=RCIM_Cubic, leaveTangent=-3.1485, tangentMode=RCTM_User, time=0, value=1}
           {interpMode=RCIM_Cubic, tangentMode=RCTM_User, time=0.5, value=0}
           {arriveTangent=2.6943, interpMode=RCIM_Cubic, leaveTangent=2.6943, tangentMode=RCTM_User, time=1, value=1}
          ]}}}
       CurveIndex=$Particles.RandomFloat
      )
     )
     Z=&Lerp_Float(
      A=15
      B=-15
      Alpha=&FloatFromCurve(
       FloatCurve=DataInterface() {propertyValues {Curve {keys [
           {arriveTangent=0.0014, interpMode=RCIM_Cubic, leaveTangent=0.0014, tangentMode=RCTM_User, time=0, value=1}
           {interpMode=RCIM_Cubic, tangentMode=RCTM_User, time=0.5, value=0.5}
           {arriveTangent=0.0577, interpMode=RCIM_Cubic, leaveTangent=0.0577, tangentMode=RCTM_User, time=1, value=0}
          ]}}}
       CurveIndex=$Particles.RandomFloat
      )
     )
    )) {isSetParameters=true}
   /SetVariables(Particles.FootOffset:Vector3f=&MakeVector(
     X=&Lerp_Float(
      A=5
      B=10
      Alpha=&FloatFromCurve(
       FloatCurve=DataInterface() {propertyValues {Curve {keys [
           {arriveTangent=-0.3783, interpMode=RCIM_Cubic, leaveTangent=-0.3783, tangentMode=RCTM_User, time=0, value=1}
           {interpMode=RCIM_Cubic, tangentMode=RCTM_User, time=0.5, value=0}
           {arriveTangent=0.3077, interpMode=RCIM_Cubic, leaveTangent=0.3077, tangentMode=RCTM_User, time=1, value=1}
          ]}}}
       CurveIndex=$Particles.RandomFloat
      )
     )
     Y=&Lerp_Float(
      A=25
      B=-5
      Alpha=&FloatFromCurve(
       FloatCurve=DataInterface() {propertyValues {Curve {keys [
           {arriveTangent=-0.3783, interpMode=RCIM_Cubic, leaveTangent=-0.3783, tangentMode=RCTM_User, time=0, value=1}
           {interpMode=RCIM_Cubic, tangentMode=RCTM_User, time=0.5, value=0}
           {arriveTangent=0.3077, interpMode=RCIM_Cubic, leaveTangent=0.3077, tangentMode=RCTM_User, time=1, value=1}
          ]}}}
       CurveIndex=$Particles.RandomFloat
      )
     )
     Z=&Lerp_Float(
      A=-15
      B=15
      Alpha=&FloatFromCurve(
       FloatCurve=DataInterface() {propertyValues {Curve {keys [
           {arriveTangent=0.899, interpMode=RCIM_Cubic, leaveTangent=0.899, tangentMode=RCTM_User, time=0, value=0}
           {arriveTangent=0.9309, interpMode=RCIM_Cubic, leaveTangent=0.9309, tangentMode=RCTM_User, time=1, value=1}
          ]}}}
       CurveIndex=$Particles.RandomFloat
      )
     )
    )) {isSetParameters=true}
   /SetVariables(Particles.SocketOffset:Vector3f=&MergeHelper_VectorSwitch(
     0=$Particles.LegOffset
     1=&Multiply_Vector(A=$Particles.LegOffset, B=<-1 -1 -1>)
     2=$Particles.ShoulderOffset
     3=&Multiply_Vector(A=$Particles.ShoulderOffset, B=<-1 -1 -1>)
     4=$Particles.ChestOffset
     5=$Particles.FootOffset
     6=&Multiply_Vector(A=$Particles.FootOffset, B=<-1 -1 1>)
     Index=$Particles.IndexType
     Num=::NewEnumerator5
    )) {isSetParameters=true}
  ]
  particleUpdateScript [
   &ParticleState()
   &SocketLocation(
    SamplePosOffset=true
    Sampled Position Offset=$Particles.SocketOffset
    Socket Index=$Particles.IndexType
    Socket Reader=$System.Socket Reader
    Socket Sampling Mode=::NewEnumerator3
    VelocitySpeedLimit=false
   )
   /SetVariables(Particles.MeshOrientation=&OrientMeshToVector(
     Look At Direction=&Add_Vector(
      A=&NormalizeVector(
       Vector To Normalize=&RotateVectorByQuaternion(Vector To Rotate=$Particles.LegOffset, Quaternion=$Output.SocketLocation.SampledRotation)
       Safe Normalize=true
       Safe Fallback Threshold=0
       Safe Fallback Vector=<0 0 0.001>
      )
      B=<0 0 0>
     )
     Reference Axis Vector=<1 0 0>
    )) {isSetParameters=true}
   &ScaleSpriteSize(
    Scale Sprite Size Mode=::NewEnumerator0
    Uniform Scale Factor=&Multiply_Float(
     A=&Multiply_Float(
      A=&FloatFromCurve(
       FloatCurve=DataInterface() {propertyValues {Curve {keys [
           {arriveTangent=5.4245, interpMode=RCIM_Cubic, leaveTangent=5.4245, tangentMode=RCTM_User, time=0, value=5}
           {arriveTangent=-3.9614, interpMode=RCIM_Cubic, leaveTangent=-3.9614, time=0.0963, value=8.5355}
           {arriveTangent=-22.2662, interpMode=RCIM_Cubic, leaveTangent=-22.2662, time=0.2702, value=3.9295}
           <at 0.3 4>
           <at 0.4 11.5>
           {arriveTangent=0.1532, interpMode=RCIM_Cubic, leaveTangent=0.1532, tangentMode=RCTM_User, time=0.5, value=2.875}
           <at 0.6 3>
           <at 0.7 11.5>
           <at 0.8 2.5>
           <at 0.9 2>
           {arriveTangent=-0.0405, interpMode=RCIM_Cubic, leaveTangent=-0.0405, tangentMode=RCTM_User, time=1, value=0.5}
          ]}}}
       Scale Curve=&Multiply_Float(A=0.1, B=1)
      )
      B=&ClampFloat(Float=&Divide_Float(
        A=&Subtract_Float(
         A=&VectorLength(Vector=&Subtract_Position(
           A=$Particles.Position
           B=&ReturnCameraPosition(Coordinate Space:ENiagaraCoordinateSpace=::Simulation, Camera Query:NiagaraDataInterfaceCamera=DataInterface())
          ))
         B=75
        )
        B=50
       ))
     )
     B=&Lerp_Float(A=0.5, B=1, Alpha=$System.ElimAlpha)
    )
   )
   &ScaleColor(
    Scale RGB=&VectorFromFloat(Value=&Lerp_Float(
      A=1
      B=10
      Alpha=&FloatFromCurve(FloatCurve=DataInterface() {propertyValues {Curve {keys [<at 0.8 0>,<at 1 1>]}}}, CurveIndex=$System.ElimAlpha)
     ))
    ScaleA=true
    ScaleRGB=true
   )
   &CameraOffset()
   &Drag(Drag=10, Use Linear Drag=true, Use Rotational Drag=true)
   &CurlNoiseForce(Noise Frequency=100, Noise Strength=250, Pan Noise=false)
   &SolveForcesAndVelocity(Clamp Velocity=false, Limit Acceleration=false)
   &DynamicMaterialParameters(
    Index 0 Param 2=&RandomRangeFloat(Minimum=-1, Maximum=1)
    Index 0 Param 3=&FloatFromCurve(FloatCurve=DataInterface() {propertyValues {Curve {keys [
         <at 0 1 RCIM_Cubic>
         {arriveTangent=-3.1701, interpMode=RCIM_Cubic, leaveTangent=-3.1701, tangentMode=RCTM_User, time=0.5612, value=0.5}
         {arriveTangent=-0.8021, interpMode=RCIM_Cubic, leaveTangent=-0.8021, tangentMode=RCTM_User, time=1, value=0}
        ]}}})
    Index 1 Param 4=&FloatFromCurve(FloatCurve=DataInterface() {propertyValues {Curve {keys [
         {arriveTangent=4.1785, interpMode=RCIM_Cubic, leaveTangent=4.1785, tangentMode=RCTM_User, time=0, value=0}
         {arriveTangent=0.9995, interpMode=RCIM_Cubic, leaveTangent=0.9995, tangentMode=RCTM_User, time=0.2223, value=0.6825}
         {arriveTangent=-0.0405, interpMode=RCIM_Cubic, leaveTangent=-0.0405, tangentMode=RCTM_User, time=1, value=1}
        ]}}})
    Index 3 Param 1=3
    Index 3 Param 2=1
    Index 3 Param 3=1
    Index 3 Param 4=1
    Module.Index0Float0Write=true
    Module.Index0Float1Write=true
    Module.Index0Float2Write=true
    Module.Index0Float3Write=true
    Module.Index1Float0Write=true
    Module.Index1Float1Write=true
    Module.Index1Float2Write=true
    Module.Index1Float3Write=true
    Module.Index3Float0Write=true
    Module.Index3Float1Write=true
    Module.Index3Float2Write=true
    Module.Index3Float3Write=true
    Write Parameter Index 1=::NewEnumerator1
    Write Parameter Index 3=::NewEnumerator1
   ) {module=DynamicMaterialParameters002}
   &SpriteFacingAndAlignment(
    Sprite Alignment=&Add_Vector(
     A=&SelectVectorFromArray(
      Direct Array Index=$Particles.IndexType
      Vector Selection Array=DataInterface() {propertyValues {FloatData [<0 0 2>, <0 0 2>, <0 0 1>, <0 0 1>, <0 0 1>]}}
     )
     B=&NormalizeVector(
      Vector To Normalize=&RotateVectorByQuaternion(Vector To Rotate=$Particles.SocketOffset, Quaternion=$Output.SocketLocation.SampledRotation)
      Safe Normalize=true
      Safe Fallback Threshold=0
      Safe Fallback Vector=<0 0 0.001>
     )
    )
    WriteAlignment=true
    WriteFacing=false
   )
   &FN_Sprite_Attributes(
    Create Position Attribute=true
    Create SpriteSize Attribute=true
    Position:NiagaraPosition=&AddVectorToPosition(Position=$Particles.Position, Vector=&RandomVector(Vector Scale=5))
    SpriteSize=&Multiply_Vector2D(A=$Particles.SpriteSize, B=&RandomRangeVector2D(Minimum=<0.5 1.25>, Maximum=<0.25 1.5>))
    Write to Module Namespace=true
   )
   &FN_Sprite_Attributes(
    Color=&ColorFromCurve(
     ColorCurve=DataInterface() {propertyValues {
       AlphaCurve {keys [<at 0 1 RCIM_Cubic>]}
       BlueCurve {keys [<at 0.4616 0.4726 RCIM_Cubic>,<at 0.4893 1 RCIM_Cubic>]}
       GreenCurve {keys [<at 0.4616 0.1 RCIM_Cubic>,<at 0.4893 0.6287 RCIM_Cubic>]}
       RedCurve {keys [<at 0.4616 1 RCIM_Cubic>,<at 0.4893 0.3 RCIM_Cubic>]}
      }}
     CurveIndex=&RandomRangeFloat(Minimum=0, Maximum=1)
    )
    Create Color Attribute=true
    Create Position Attribute=true
    Create SpriteSize Attribute=true
    Position:NiagaraPosition=&AddVectorToPosition(Position=$Particles.Position, Vector=&RandomVector(Vector Scale=5))
    SpriteSize=&Multiply_Vector2D(A=$Particles.SpriteSize, B=&RandomRangeVector2D(Minimum=<0.5 0.5>, Maximum=<0.25 0.7>))
    Write to Module Namespace=true
   ) {module=FN_Sprite_Attributes001}
   &FN_Sprite_Attributes(
    Color=&MultiplyLinearColor(
     Linear Color=$Particles.Color
     Scale Factor=&MakeVector4FromVectorAndFloat(
      "Vector (RGB)"=<1 1 1>
      "Float (Alpha)"=&FloatFromCurve(FloatCurve=DataInterface() {propertyValues {Curve {keys [<at 0 1>,<at 0.5 0>]}}})
     )
    )
    Create Color Attribute=true
    Create SpriteSize Attribute=true
    SpriteSize=&Multiply_Vector2D(A=$Particles.SpriteSize, B=<4 2.5>)
    Write to Module Namespace=true
   ) {module=FN_Sprite_Attributes002}
   &FN_Sprite_Attributes(
    Camera Offset=10
    Color=&MultiplyLinearColor(
     Linear Color=&Multiply_LinearColorByFloat(Linear Color=<rgba 2 2 2 1>, Float=$System.ElimAlpha)
     Scale Factor=&MakeVector4_FromLinearColor(LinearColor=$Particles.Color)
    )
    Create Camera Offset Attribute=true
    Create Color Attribute=true
    Create SpriteSize Attribute=true
    SpriteSize=&Vector2DFromFloat(Value=&Multiply_Float(
      A=&FloatFromCurve(
       FloatCurve=DataInterface() {propertyValues {Curve {keys [
           {arriveTangent=5.4245, interpMode=RCIM_Cubic, leaveTangent=5.4245, tangentMode=RCTM_User, time=0, value=5}
           {arriveTangent=-3.9614, interpMode=RCIM_Cubic, leaveTangent=-3.9614, time=0.0963, value=8.5355}
           {arriveTangent=-22.2662, interpMode=RCIM_Cubic, leaveTangent=-22.2662, time=0.2702, value=3.9295}
           <at 0.3 4>
           <at 0.4 11.5>
           {arriveTangent=0.1532, interpMode=RCIM_Cubic, leaveTangent=0.1532, tangentMode=RCTM_User, time=0.5, value=2.875}
           <at 0.6 3>
           <at 0.7 11.5>
           <at 0.8 2.5>
           <at 0.9 2>
           {arriveTangent=-0.0405, interpMode=RCIM_Cubic, leaveTangent=-0.0405, tangentMode=RCTM_User, time=1, value=0.5}
          ]}}}
       Scale Curve=&Multiply_Float(A=0.1, B=1)
      )
      B=&Multiply_Float(
       A=35
       B=&SelectFloatFromArray(Direct Array Index=$Particles.IndexType, Float Selection Array=DataInterface() {propertyValues {FloatData [1,1,1,1,1,0,0]}})
      )
     ))
    Write to Module Namespace=true
   ) {module=FN_Sprite_Attributes003}
   /SetVariables(Particles.GlowDecalDynMat:Vector4f=&MakeVector4(
     X=4
     Y=0
     Z=&FloatFromCurve(FloatCurve=DataInterface() {propertyValues {Curve {keys [<at 0 -0.1 RCIM_Cubic>, {interpMode=RCIM_Cubic, leaveTangent=1.6586, time=1, value=0.5}]}}})
     W=0
    )) {isSetParameters=true}
  ]
  rendererClasses [&NiagaraSpriteRendererProperties]
  renderers [
   {rendererClass=&NiagaraSpriteRendererProperties}
   {rendererClass=&NiagaraSpriteRendererProperties, rendererIndex=1}
   {rendererClass=&NiagaraSpriteRendererProperties, rendererIndex=2}
   {rendererClass=&NiagaraSpriteRendererProperties, rendererIndex=3}
   {rendererClass=&NiagaraSpriteRendererProperties, rendererIndex=4}
  ]
 }}
summary {
 systemName=NS_Character_PawJasmineSword_BodyIdle
 emitters [{emitterName=NE_GlitchSprites, rendererClasses [&NiagaraSpriteRendererProperties]}]
 userVariables [
  User.bisFootHidden {type=NiagaraBool}
  User.GameTime {type=NiagaraFloat, defaultValue=-100, description=""}
  User.IsReactive {type=NiagaraFloat, defaultValue=0, description=""}
  User.KillCount {type=NiagaraInt32}
  User.VariantIndex {type=NiagaraFloat, defaultValue=0, description=""}
 ]
}
