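# Idle head effect for the PawJasmineSword character, placing glowing hair meshes and glitch sprites at two socket-relative points on the head.
# NE_HairGlows bursts variant-dependent, slowly spinning, pulsing meshes that shrink away when the head is hidden or hood is up.
# NE_GlitchSprites feeds two spawn groups into rings around those points, with curl noise, flickering size and camera-distance scaling.
# System ElimAlpha and ElimFlash values intensify brightness, size and spin.
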NiagaraSystem /BRCosmetics/Characters/Player/Female/Medium/Bodies/F_MED_PawJasmineSword/FX/NS_Character_PawJasmineSword_HeadIdle
dependencies {
 usedDataInterfaces [
  &NiagaraDataInterfaceColorCurve
  &NiagaraDataInterfaceArrayFloat3
  &NiagaraDataInterfaceCurve
  &NiagaraDataInterfaceCamera
  &NiagaraDataInterfaceSocketReader
 ]
 usedDynamicInputs [
  &SetBoolByFloatComparison
  &Boolean_AndOperation
  &SetBoolByIntComparison
  &Boolean_NotOperation
  &Boolean_OrOperation
 ]
 usedModules [
  &self:SystemSpawnScript.SystemScriptSource.SystemScriptGraph.NiagaraNodeAssignment_0.SetVariables_#1
  &self:SystemSpawnScript.SystemScriptSource.SystemScriptGraph.NiagaraNodeAssignment_3.SetVariables_#2
  &SystemState
  /Game/Effects/Niagara/Modules/Logic/UpdateValueStateChange
  &Timeline
  &self:SystemSpawnScript.SystemScriptSource.SystemScriptGraph.NiagaraNodeAssignment_7.SetVariables_#3
  &self:SystemSpawnScript.SystemScriptSource.SystemScriptGraph.NiagaraNodeAssignment_2.SetVariables_#4
  &self:SystemSpawnScript.SystemScriptSource.SystemScriptGraph.NiagaraNodeAssignment_4.SetVariables_#5
  &self:SystemSpawnScript.SystemScriptSource.SystemScriptGraph.NiagaraNodeAssignment_1.SetVariables_#6
  &self:SystemSpawnScript.SystemScriptSource.SystemScriptGraph.NiagaraNodeAssignment_5.SetVariables_#7
  &self:SystemSpawnScript.SystemScriptSource.SystemScriptGraph.NiagaraNodeAssignment_6.SetVariables_#8
 ]
}
emitters {
 NE_GlitchSprites {
  emitterUpdateScript [
   &EmitterState()
   &SpawnRate(SpawnRate=&Multiply_Float(A=20, B=$System.SpawnMultiplier), Use Spawn Probability=false)
   &SpawnRate(SpawnGroup=1, SpawnRate=&Multiply_Float(A=20, B=$System.SpawnMultiplier), Use Spawn Probability=false) {module=SpawnRate001}
  ]
  particleSpawnScript [
   /SetVariables(Particles.RandomFloat=&RandomRangeFloat(Minimum=0, Maximum=1)) {isSetParameters=true}
   &InitializeParticle(
    Color=&ColorFromCurve(
     ColorCurve=DataInterface() {propertyValues {
       AlphaCurve {keys [<at 0 1 RCIM_Cubic>]}
       BlueCurve {keys [<at 0.4616 0 RCIM_Cubic>,<at 0.4893 0 RCIM_Cubic>]}
       GreenCurve {keys [<at 0.4616 0.2438 RCIM_Cubic>,<at 0.4893 0.1091 RCIM_Cubic>]}
       RedCurve {keys [<at 0.4616 2 RCIM_Cubic>,<at 0.4893 1.5 RCIM_Cubic>]}
      }}
     CurveIndex=&RandomRangeFloat(Minimum=0, Maximum=1)
    )
    Lifetime Max=0.7
    Lifetime Min=0.5
    Lifetime Mode=::NewEnumerator1
    Mesh Scale Mode=::NewEnumerator3
    Mesh Uniform Scale=0.1
    Sprite Size=&Multiply_Vector2D_ByFloat(
     Float=&RandomRangeFloat(Minimum=3, Maximum=5)
     Vector2D=&Lerp_Vector2(A=<1 0.5>, B=<0.5 0.2>, Alpha=$Particles.RandomFloat)
    )
    Sprite Size Mode=::NewEnumerator4
    UsePositionOffset=true
    Write Material Random=true
   ) {module=InitializeParticle001}
   /SetVariables(Particles.IndexType:NiagaraInt32=&SpawnGroup()) {isSetParameters=true}
   &DynamicMaterialParameters(
    Index 0 Param 4=&RandomRangeFloat(Minimum=-1, Maximum=0.1)
    Index 1 Param 1=&RandomRangeFloat(Minimum=-100, Maximum=100)
    Module.Param0WriteEnabled=true
    Module.Param0WriteEnabled1=true
    Module.Param1WriteEnabled=true
    Module.Param1WriteEnabled1=true
    Module.Param2WriteEnabled=true
    Module.Param2WriteEnabled1=true
    Module.Param3WriteEnabled=true
    Module.Param3WriteEnabled1=true
    Write Parameter Index 1=true
   ) {module=DynamicMaterialParameters001}
   /SetVariables(Particles.HairGlowOffset:Vector3f=&SelectVectorFromArray(
     Direct Array Index=&SpawnGroup()
     Vector Selection Array=DataInterface() {propertyValues {FloatData [<11 -4 -7>, <11 -4 7>]}}
    )) {isSetParameters=true}
   /SetVariables(Particles.HairOffset:Vector3f=&RotateVectorByQuaternion(
     Vector To Rotate=&MakeVector(
      X=0
      Y=&Cosine(Normalized Angle=$Particles.RandomFloat, Period=1, Scale=6, Bias=0)
      Z=&Sine(Normalized Angle=$Particles.RandomFloat, Period=1, Scale=6, Bias=0)
     )
     Quaternion=&OrientMeshToVector(
      Look At Direction=&NormalizeVector(Vector To Normalize=$Particles.HairGlowOffset, Safe Normalize=true, Safe Fallback Threshold=0, Safe Fallback Vector=<0 0 0.001>)
      Reference Axis Vector=<1 0 0>
     )
    )) {isSetParameters=true}
   /SetVariables(Particles.SocketOffset:Vector3f=&Add_Vector(A=$Particles.HairOffset, B=$Particles.HairGlowOffset)) {isSetParameters=true}
  ]
  particleUpdateScript [
   &ParticleState()
   &SocketLocation(
    SamplePosOffset=true
    Sampled Position Offset=$Particles.SocketOffset
    Socket Reader=$System.Socket Reader
    Socket Sampling Mode=::NewEnumerator3
    VelocitySpeedLimit=false
   )
   &ScaleSpriteSize(
    Scale Sprite Size Mode=::NewEnumerator0
    Uniform Scale Factor=&Multiply_Float(
     A=&Multiply_Float(
      A=&Multiply_Float(
       A=&FloatFromCurve(
        FloatCurve=DataInterface() {propertyValues {Curve {keys [
            {arriveTangent=5.4245, interpMode=RCIM_Cubic, leaveTangent=5.4245, tangentMode=RCTM_User, time=0, value=5}
            {arriveTangent=-3.9614, interpMode=RCIM_Cubic, leaveTangent=-3.9614, time=0.0963, value=8.5355}
            {arriveTangent=-22.2662, interpMode=RCIM_Cubic, leaveTangent=-22.2662, time=0.2702, value=3.9295}
            <at 0.3 4>
            <at 0.4 11.5>
            {arriveTangent=0.1532, interpMode=RCIM_Cubic, leaveTangent=0.1532, tangentMode=RCTM_User, time=0.5, value=2.875}
            <at 0.6 3>
            <at 0.7 11.5>
            <at 0.8 2.5>
            <at 0.9 2>
            {arriveTangent=-0.0405, interpMode=RCIM_Cubic, leaveTangent=-0.0405, tangentMode=RCTM_User, time=1, value=0.5}
           ]}}}
        Scale Curve=&Multiply_Float(A=0.1, B=1)
       )
       B=&ClampFloat(Float=&Divide_Float(
         A=&Subtract_Float(
          A=&VectorLength(Vector=&Subtract_Position(
            A=$Particles.Position
            B=&ReturnCameraPosition(Coordinate Space:ENiagaraCoordinateSpace=::Simulation, Camera Query:NiagaraDataInterfaceCamera=DataInterface())
           ))
          B=5
         )
         B=10
        ))
      )
      B=&Lerp_Float(A=0.5, B=1, Alpha=$System.ElimAlpha)
     )
     B=1
    )
   )
   &ScaleColor(
    Scale RGB=&VectorFromFloat(Value=&Lerp_Float(
      A=1
      B=10
      Alpha=&FloatFromCurve(FloatCurve=DataInterface() {propertyValues {Curve {keys [<at 0.8 0>,<at 1 1>]}}}, CurveIndex=$System.ElimAlpha)
     ))
    ScaleA=true
    ScaleRGB=true
   )
   &Drag(Drag=10, Use Linear Drag=true, Use Rotational Drag=true)
   &CurlNoiseForce(Noise Frequency=100, Noise Strength=250, Pan Noise=false)
   &SolveForcesAndVelocity(Clamp Velocity=false, Limit Acceleration=false)
   &DynamicMaterialParameters(
    Index 0 Param 2=&RandomRangeFloat(Minimum=-1, Maximum=1)
    Index 0 Param 3=&FloatFromCurve(FloatCurve=DataInterface() {propertyValues {Curve {keys [
         <at 0 1 RCIM_Cubic>
         {arriveTangent=-3.1701, interpMode=RCIM_Cubic, leaveTangent=-3.1701, tangentMode=RCTM_User, time=0.5612, value=0.5}
         {arriveTangent=-0.8021, interpMode=RCIM_Cubic, leaveTangent=-0.8021, tangentMode=RCTM_User, time=1, value=0}
        ]}}})
    Index 1 Param 4=&FloatFromCurve(FloatCurve=DataInterface() {propertyValues {Curve {keys [
         {arriveTangent=4.1785, interpMode=RCIM_Cubic, leaveTangent=4.1785, tangentMode=RCTM_User, time=0, value=0}
         {arriveTangent=0.9995, interpMode=RCIM_Cubic, leaveTangent=0.9995, tangentMode=RCTM_User, time=0.2223, value=0.6825}
         {arriveTangent=-0.0405, interpMode=RCIM_Cubic, leaveTangent=-0.0405, tangentMode=RCTM_User, time=1, value=1}
        ]}}})
    Index 3 Param 1=3
    Index 3 Param 2=1
    Index 3 Param 3=1
    Index 3 Param 4=1
    Module.Index0Float0Write=true
    Module.Index0Float1Write=true
    Module.Index0Float2Write=true
    Module.Index0Float3Write=true
    Module.Index1Float0Write=true
    Module.Index1Float1Write=true
    Module.Index1Float2Write=true
    Module.Index1Float3Write=true
    Module.Index3Float0Write=true
    Module.Index3Float1Write=true
    Module.Index3Float2Write=true
    Module.Index3Float3Write=true
    Write Parameter Index 1=::NewEnumerator1
    Write Parameter Index 3=::NewEnumerator1
   ) {module=DynamicMaterialParameters002}
   &SpriteFacingAndAlignment(
    Sprite Alignment=&Add_Vector(
     A=<0 0 0>
     B=&NormalizeVector(
      Vector To Normalize=&RotateVectorByQuaternion(
       Vector To Rotate=&Multiply_Vector(A=$Particles.SocketOffset, B=<1 1 1>)
       Quaternion=$Output.SocketLocation.SampledRotation
      )
      Safe Normalize=true
      Safe Fallback Threshold=0
      Safe Fallback Vector=<0 0 0.001>
     )
    )
    WriteAlignment=true
    WriteFacing=false
   )
   &FN_Sprite_Attributes(
    Create Position Attribute=true
    Create SpriteSize Attribute=true
    Position:NiagaraPosition=&AddVectorToPosition(Position=$Particles.Position, Vector=&RandomVector(Vector Scale=1))
    SpriteSize=&Multiply_Vector2D(A=$Particles.SpriteSize, B=&RandomRangeVector2D(Minimum=<0.5 1.25>, Maximum=<0.25 1.5>))
    Write to Module Namespace=true
   )
   &FN_Sprite_Attributes(
    Color=&ColorFromCurve(
     ColorCurve=DataInterface() {propertyValues {
       AlphaCurve {keys [<at 0 1 RCIM_Cubic>]}
       BlueCurve {keys [<at 0.4616 0.4726 RCIM_Cubic>,<at 0.4893 1 RCIM_Cubic>]}
       GreenCurve {keys [<at 0.4616 0.1 RCIM_Cubic>,<at 0.4893 0.6287 RCIM_Cubic>]}
       RedCurve {keys [<at 0.4616 1 RCIM_Cubic>,<at 0.4893 0.3 RCIM_Cubic>]}
      }}
     CurveIndex=&RandomRangeFloat(Minimum=0, Maximum=1)
    )
    Create Color Attribute=true
    Create Position Attribute=true
    Create SpriteSize Attribute=true
    Position:NiagaraPosition=&AddVectorToPosition(Position=$Particles.Position, Vector=&RandomVector(Vector Scale=1))
    SpriteSize=&Multiply_Vector2D(A=$Particles.SpriteSize, B=&RandomRangeVector2D(Minimum=<0.5 0.5>, Maximum=<0.25 0.7>))
    Write to Module Namespace=true
   ) {module=FN_Sprite_Attributes001}
   &FN_Sprite_Attributes(
    Color=&MultiplyLinearColor(
     Linear Color=$Particles.Color
     Scale Factor=&MakeVector4FromVectorAndFloat(
      "Vector (RGB)"=<1 1 1>
      "Float (Alpha)"=&FloatFromCurve(FloatCurve=DataInterface() {propertyValues {Curve {keys [<at 0 1>,<at 0.5 0>]}}})
     )
    )
    Create Color Attribute=true
    Create SpriteSize Attribute=true
    SpriteSize=&Multiply_Vector2D(A=$Particles.SpriteSize, B=<4 2.5>)
    Write to Module Namespace=true
   ) {module=FN_Sprite_Attributes002}
   &FN_Sprite_Attributes(
    Camera Offset=10
    Color=&MultiplyLinearColor(
     Linear Color=&Multiply_LinearColorByFloat(Linear Color=<rgba 2 2 2 1>, Float=$System.ElimAlpha)
     Scale Factor=&MakeVector4_FromLinearColor(LinearColor=$Particles.Color)
    )
    Create Camera Offset Attribute=true
    Create Color Attribute=true
    Create SpriteSize Attribute=true
    SpriteSize=&Vector2DFromFloat(Value=&Multiply_Float(
      A=&FloatFromCurve(
       FloatCurve=DataInterface() {propertyValues {Curve {keys [
           {arriveTangent=5.4245, interpMode=RCIM_Cubic, leaveTangent=5.4245, tangentMode=RCTM_User, time=0, value=5}
           {arriveTangent=-3.9614, interpMode=RCIM_Cubic, leaveTangent=-3.9614, time=0.0963, value=8.5355}
           {arriveTangent=-22.2662, interpMode=RCIM_Cubic, leaveTangent=-22.2662, time=0.2702, value=3.9295}
           <at 0.3 4>
           <at 0.4 11.5>
           {arriveTangent=0.1532, interpMode=RCIM_Cubic, leaveTangent=0.1532, tangentMode=RCTM_User, time=0.5, value=2.875}
           <at 0.6 3>
           <at 0.7 11.5>
           <at 0.8 2.5>
           <at 0.9 2>
           {arriveTangent=-0.0405, interpMode=RCIM_Cubic, leaveTangent=-0.0405, tangentMode=RCTM_User, time=1, value=0.5}
          ]}}}
       Scale Curve=&Multiply_Float(A=0.1, B=1)
      )
      B=15
     ))
    Write to Module Namespace=true
   ) {module=FN_Sprite_Attributes003}
   /SetVariables(Particles.GlowDecalDynMat:Vector4f=&MakeVector4(
     X=4
     Y=0
     Z=&FloatFromCurve(FloatCurve=DataInterface() {propertyValues {Curve {keys [<at 0 -0.1 RCIM_Cubic>, {interpMode=RCIM_Cubic, leaveTangent=1.6586, time=1, value=0.5}]}}})
     W=0
    )) {isSetParameters=true}
  ]
  rendererClasses [&NiagaraSpriteRendererProperties]
  renderers [
   {rendererClass=&NiagaraSpriteRendererProperties}
   {rendererClass=&NiagaraSpriteRendererProperties, rendererIndex=1}
   {rendererClass=&NiagaraSpriteRendererProperties, rendererIndex=2}
   {rendererClass=&NiagaraSpriteRendererProperties, rendererIndex=3}
   {rendererClass=&NiagaraSpriteRendererProperties, rendererIndex=4}
  ]
 }
 NE_HairGlows {
  emitterSpawnScript [/SetVariables() {isSetParameters=true}]
  emitterUpdateScript [
   &EmitterState(
    Inactive Response=::NewEnumerator1
    Life Cycle Mode=::NewEnumerator1
    Loop Behavior=::NewEnumerator1
    Loop Duration Mode=::NewEnumerator1
    UseLoopDelay=false
   )
   &SpawnBurst_Instantaneous(
    Spawn Count=&MergeHelper_IntSwitch(
     0=0
     1=2
     2=?
     3=?
     4=?
     5=?
     Index=&MakeIntFromFloat(Float=$User.VariantIndex, Conversion Mode=::NewEnumerator1)
     Num=::NewEnumerator0
    )
    Use Spawn Probability=false
    UseLoopCountLimit=false
   )
   /SetVariables(Emitter.CustomTimer:NiagaraFloat=&Add_Float(A=$Emitter.CustomTimer, B=&Lerp_Float(A=0.03, B=0.06, Alpha=$System.ElimAlpha))) {isSetParameters=true}
   /SetVariables(Emitter.Oscillation:NiagaraFloat=&Waveform(
     Waveform Count=::NewEnumerator1
     "Waveform [1]"=::NewEnumerator0
     "[1] Phase Scale"=1.3541
     "[1] Frequency"=1
     "[1] Amplitude Min/Max"=<0 1>
     "[2] Average"=false
     "[2] Input Position"=0
     "[2] Phase"=$Emitter.Age
     "[2] Phase Scale"=1
     "[2] Compound Frequency"=2
     "[2] Frequency"=0.6589
     "[2] Amplitude Min/Max"=<0 1>
     "[2] Scale"=1
     "[2] Bias"=0
     "[2] Exponent"=1
     "[2] Pulse Duration"=0.1
     "[2] Clip Triangle Min/Max"=<0 0>
     "[2] Random Seed"=0
    )) {isSetParameters=true}
  ]
  particleSpawnScript [
   &InitializeParticle(
    Color Mode=::NewEnumerator0
    Lifetime=5
    Mesh Scale Mode=::NewEnumerator3
    Mesh Uniform Scale=0.225
    Sprite Size=<6 5>
    Sprite Size Mode=::NewEnumerator4
    UsePositionOffset=false
    Write Material Random=true
   )
   &PartitionParticles(ID to Partition:NiagaraInt32=&ReturnExecIndex())
   /SetVariables(Particles.HairGlowOffset:Vector3f=&SelectVectorFromArray(
     Direct Array Index=$Particles.PartitionParticles.Partition
     Vector Selection Array=DataInterface() {propertyValues {FloatData [<11 -4 -7>, <11 -4 7>]}}
    )) {isSetParameters=true}
   &InitialMeshOrientation(
    Mesh Orientation Mode=::NewEnumerator7
    Rotation=&MaskVector(Mask Source=1, Mask Parameter=$Particles.PartitionParticles.Partition, Mask=0, Masked Vector=<0 0 0>, Passthrough Vector=<0 0 0>)
    Use Rotation Vector=true
   )
  ]
  particleUpdateScript [
   &ParticleState(Kill Particles When Lifetime Has Elapsed=false)
   &ScaleSpriteSize(
    Scale Factor=&Multiply_Vector2D_ByFloat(
     Float=&Multiply_Float(A=&Lerp_Float(A=1, B=1.1, Alpha=$Emitter.Oscillation), B=1)
     Vector2D=&Multiply_Vector2D(
      A=&Lerp_Vector2(A=<1 1>, B=<0.7 2>, Alpha=$System.ElimFlash)
      B=&MakeVector2D(X=$System.FadeIn, Y=1)
     )
    )
    Scale Sprite Size Mode=::NewEnumerator2
   )
   &ScaleMeshSize(Scale Factor=&VectorFromFloat(Value=&Multiply_Float(
      A=&Multiply_Float(
       A=&Lerp_Float(A=0, B=1, Alpha=$System.FadeIn)
       B=&Multiply_Float(
        A=&Lerp_Float(A=1, B=1.75, Alpha=$System.ElimFlash)
        B=&Lerp_Float(A=1, B=1.1, Alpha=$Emitter.Oscillation)
       )
      )
      B=&MakeCustomFloatFromBool(Bool=$User.bisHeadHidden, True Float=0, False Float=&Lerp_Float(A=1, B=0, Alpha=$User.HoodUp))
     )))
   &SocketLocation(
    Orientation Sampling=::NewEnumerator1
    SamplePosOffset=true
    Sampled Position Offset=$Particles.HairGlowOffset
    Socket Reader=$System.Socket Reader
    Socket Sampling Mode=::NewEnumerator3
    VelocitySpeedLimit=false
   )
   /SetVariables(Particles.MeshOrientation=&MultiplyQuaternion(
     Quaternion A=$Particles.Initial.MeshOrientation
     Quaternion B=&MultiplyQuaternion(
      Quaternion A=&OrientMeshToVector(
       Look At Direction=&NormalizeVector(
        Vector To Normalize=&RotateVectorByQuaternion(Vector To Rotate=$Particles.HairGlowOffset, Quaternion=$Output.SocketLocation.SampledRotation)
        Safe Normalize=true
        Safe Fallback Threshold=0
        Safe Fallback Vector=<0 0 0.001>
       )
       Reference Axis Vector=<1 0 0>
      )
      Quaternion B=&MakeQuaternion(
       Quaternion From=0
       Axis=<1 0 0>
       Angle=&Multiply_Float(A=$Emitter.CustomTimer, B=1)
       From=?
       To=?
       Angle Type=::NewEnumerator1
       Normalize Vectors=false
      )
     )
    )) {isSetParameters=true}
   &Color(Color=&Multiply_LinearColorByFloat(
     Linear Color=&ColorFromCurve(
      ColorCurve=DataInterface() {propertyValues {
        AlphaCurve {keys [<at 0 1 RCIM_Cubic>]}
        BlueCurve {keys [<at 0 0 RCIM_Cubic>,<at 1 0 RCIM_Cubic>]}
        GreenCurve {keys [<at 0 0.1967 RCIM_Cubic>,<at 1 0.9833 RCIM_Cubic>]}
        RedCurve {keys [<at 0 0.6 RCIM_Cubic>,<at 1 3 RCIM_Cubic>]}
       }}
      CurveIndex=$System.ElimFlash
     )
     Float=&FloatFromCurve(FloatCurve=DataInterface() {propertyValues {Curve {keys [<at 0 0>,<at 0.5 1>]}}}, CurveIndex=$System.FadeIn)
    ))
   &DynamicMaterialParameters(
    Index 3 Param 1=1
    Index 3 Param 2=1
    Index 3 Param 3=1
    Index 3 Param 4=1
    Module.Index0Float0Write=true
    Module.Index0Float1Write=true
    Module.Index0Float2Write=true
    Module.Index0Float3Write=true
    Module.Index3Float0Write=true
    Module.Index3Float1Write=true
    Module.Index3Float2Write=true
    Module.Index3Float3Write=true
    Write Parameter Index 3=::NewEnumerator1
   )
   &FN_Mesh_Attributes(
    Color=&MultiplyLinearColor(
     Linear Color=$Particles.Color
     Scale Factor=&MakeVector4FromVectorAndFloat(
      "Vector (RGB)"=&MakeVectorFromLinearColorRGB(Color=<rgba 0.1771 0.1016 0.0914 1>)
      "Float (Alpha)"=&MakeFloatFromLinearColor(LinearColor=$Particles.Color, Channel:ENiagaraFloatFromLinearColorOptions=::NewEnumerator3)
     )
    )
    Create Color Attribute=true
    Create Scale Attribute=true
    Scale=&Multiply_Vector(A=$Particles.Scale, B=<1 1 1>)
   )
   &SpriteFacingAndAlignment(Sprite Alignment=<0 1 -0.1>, WriteAlignment=true, WriteFacing=true)
   &FN_Mesh_Attributes(
    Color=&MultiplyLinearColor(
     Linear Color=$Particles.Color
     Scale Factor=&MakeVector4FromVectorAndFloat(
      "Vector (RGB)"=&MakeVectorFromLinearColorRGB(Color=<rgba 6 0.7125 0 1>)
      "Float (Alpha)"=&MakeFloatFromLinearColor(LinearColor=$Particles.Color, Channel:ENiagaraFloatFromLinearColorOptions=::NewEnumerator3)
     )
    )
    Create Color Attribute=true
    Create Scale Attribute=true
    Scale=&Multiply_Vector(A=$Particles.Scale, B=<1 1 1.25>)
    Write to Module Namespace=true
   ) {module=FN_Mesh_Attributes001}
   &FN_Mesh_Attributes(
    Color=&MultiplyLinearColor(
     Linear Color=$Particles.Color
     Scale Factor=&MakeVector4FromVectorAndFloat(
      "Vector (RGB)"=&MakeVectorFromLinearColorRGB(Color=<rgba 20 20 20 1>)
      "Float (Alpha)"=&MakeFloatFromLinearColor(LinearColor=$Particles.Color, Channel:ENiagaraFloatFromLinearColorOptions=::NewEnumerator3)
     )
    )
    Create Color Attribute=true
    Create Scale Attribute=true
    Scale=&Multiply_Vector(A=$Particles.Scale, B=<1 1 1>)
    Write to Module Namespace=true
   ) {module=FN_Mesh_Attributes003}
   /SetVariables(Particles.SplineGlowPos:NiagaraPosition=&AddVectorToPosition(
     Position=$Particles.Position
     Vector=&RotateVectorByQuaternion(Vector To Rotate=<0 -0.25 0>, Quaternion=$Output.SocketLocation.SampledRotation)
    )) {isSetParameters=true}
  ]
  rendererClasses [&NiagaraMeshRendererProperties]
  renderers [
   {rendererClass=&NiagaraMeshRendererProperties, rendererIndex=0}
   {rendererClass=&NiagaraMeshRendererProperties, rendererIndex=1}
  ]
 }
}
summary {
 systemName=NS_Character_PawJasmineSword_HeadIdle
 emitters [
  {emitterName=NE_GlitchSprites, rendererClasses [&NiagaraSpriteRendererProperties]}
  {emitterName=NE_HairGlows, rendererClasses [&NiagaraMeshRendererProperties]}
 ]
 userVariables [
  User.bisHeadHidden {type=NiagaraBool}
  User.GameTime {type=NiagaraFloat, defaultValue=-100, description=""}
  User.HoodUp {type=NiagaraFloat, defaultValue=0, description=""}
  User.IsReactive {type=NiagaraFloat, defaultValue=0, description=""}
  User.KillCount {type=NiagaraInt32}
  User.VariantIndex {type=NiagaraFloat, defaultValue=0, description=""}
 ]
}
