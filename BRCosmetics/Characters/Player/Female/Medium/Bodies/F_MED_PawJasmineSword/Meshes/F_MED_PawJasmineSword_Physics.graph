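# Physics asset for the F_MED_PawJasmineSword female medium player body, set up for the motion of its back cloth panels.
# Left, middle and right chains each link two box bodies to a small capsule at the base bone.
# Constraints keep twist and second swing tight, leaving mainly a one-axis swing that widens at the lower segment.
# Capsules on the pelvis and both thighs, with no constraints, provide body shapes for the cloth to collide with.
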PhysicsAsset /BRCosmetics/Characters/Player/Female/Medium/Bodies/F_MED_PawJasmineSword/Meshes/F_MED_PawJasmineSword_Physics
bodies {
 backCloth_l_deform [{
   shapeType=Box
   extentX=7.172
   extentY=2.6212
   extentZ=12.1477
   shapeName=backCloth_l_deform_box
   center=<3.6096 0.5168 -0.1028>
   rotation=<rot 0.0002 0.0003 11.25>
  }]
 backCloth_l_deform_2 [{
   shapeType=Box
   extentX=6.8747
   extentY=2.0706
   extentZ=11.6926
   shapeName=backCloth_l_deform_2_box
   center=<3.3504 0.534 -0.1062>
   rotation=<rot 0.0001 0.0002 11.25>
  }]
 backCloth_l_deform_base [{shapeType=Capsule, length=0.505, radius=0.505, shapeName=backCloth_l_deform_base_capsule, rotation=<rot 0 -90 0>}]
 backCloth_m_deform [{
   shapeType=Box
   extentX=7.172
   extentY=2.6212
   extentZ=15.1874
   shapeName=backCloth_m_deform_box
   center=<3.6085 0.6063 -0.0561>
   rotation=<rot 2.8125 0.0002 0>
  }]
 backCloth_m_deform_2 [{
   shapeType=Box
   extentX=6.8747
   extentY=2.0706
   extentZ=15.5363
   shapeName=backCloth_m_deform_2_box
   center=<3.3504 0 0>
   rotation=<rot 2.8125 0.0002 0>
  }]
 backCloth_m_deform_base [{shapeType=Capsule, length=0.505, radius=0.505, shapeName=backCloth_m_deform_base_capsule, rotation=<rot 0 -90 0>}]
 backCloth_r_deform [{
   shapeType=Box
   extentX=7.172
   extentY=2.6212
   extentZ=12.1477
   shapeName=backCloth_r_deform_box
   center=<3.6096 0.5111 0.1281>
   rotation=<rot -0.0004 0.0007 -14.0625>
  }]
 backCloth_r_deform_2 [{
   shapeType=Box
   extentX=6.8747
   extentY=2.0706
   extentZ=11.6926
   shapeName=backCloth_r_deform_2_box
   center=<3.3504 0.1443 0.0438>
   rotation=<rot -0.0001 0.0002 -16.875>
  }]
 backCloth_r_deform_base [{shapeType=Capsule, length=0.505, radius=0.505, shapeName=backCloth_r_deform_base_capsule, rotation=<rot 0 -90 0>}]
 pelvis [{shapeType=Capsule, length=9.1073, radius=10.843, center=<-1.7606 -0.6482 0.2525>, rotation=<rot 0 -90 0>}]
 thigh_l [{
   shapeType=Capsule
   length=10.6156
   radius=10.5737
   shapeName=thigh_l_capsule
   center=<-11.4118 -2.0536 2.1996>
   rotation=<rot 87.1875 179.9996 179.9998>
  }]
 thigh_r [{shapeType=Capsule, length=9.7791, radius=10.6121, shapeName=thigh_r_capsule, center=<11.3224 2.3843 -1.2461>, rotation=<rot 90 0 0>}]
}
constraints [
 {bone1Name=backCloth_r_deform, bone2Name=backCloth_r_deform_base, swing1LimitDegrees=24, swing2LimitDegrees=3, twistLimitDegrees=2}
 {bone1Name=backCloth_m_deform, bone2Name=backCloth_m_deform_base, swing1LimitDegrees=24, swing2LimitDegrees=3, twistLimitDegrees=2}
 {bone1Name=backCloth_l_deform, bone2Name=backCloth_l_deform_base, swing1LimitDegrees=24, swing2LimitDegrees=3, twistLimitDegrees=2}
 {bone1Name=backCloth_r_deform_2, bone2Name=backCloth_r_deform, swing1LimitDegrees=36, swing2LimitDegrees=3, twistLimitDegrees=2}
 {bone1Name=backCloth_m_deform_2, bone2Name=backCloth_m_deform, swing1LimitDegrees=36, swing2LimitDegrees=3, twistLimitDegrees=2}
 {bone1Name=backCloth_l_deform_2, bone2Name=backCloth_l_deform, swing1LimitDegrees=36, swing2LimitDegrees=3, twistLimitDegrees=2}
]
