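# Post-process-style cosmetic Animation Blueprint for the F_MED_PawJasmineSword body, copying the attached parent's pose through a Control Rig.
# Additive offsets move attach_backpack based on spine bend, one specific backpack mesh and hood state read from either face-accessory AnimBP.
# A RigidBody node then simulates secondary motion using gravity from movement-state wind and gravity settings.
# Updates also feed material WPO parameters and enable cloth-wind strength during skydiving or parachuting.
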AnimBlueprint /BRCosmetics/Characters/Player/Female/Medium/Bodies/F_MED_PawJasmineSword/Meshes/F_MED_PawJasmineSword_AnimBP
target_skeleton=&/F_MED_PawJasmineSword_Skeleton
graphs [
 AnimGraph {nodes [
   n1 = Control Rig(Source=Copy Pose From Mesh() {bUseAttachedParent=true}) {
    alphaBoolBlend {blendOption=Linear}
    controlRigAssetReference {blueprintRigClass=&/F_MED_PawJasmineSword_CtrlRig.F_MED_PawJasmineSword_CtrlRig_C}
   }
   n2 = "Transform (Modify) Bone - Bone: attach_backpack"(
    ComponentPose=Local To Component(LocalPose=@n1)
    Alpha=Get back_transform() -> (back_transform:double)
    Translation="0,-5,0"
    Rotation="0,0,15"
    Scale="1,1,1"
   ) {
    rotationMode=BMM_Additive
    rotationSpace=BCS_BoneSpace
    translationMode=BMM_Additive
    translationSpace=BCS_ParentBoneSpace
    alphaBoolBlend {blendOption=Linear}
    boneToModify {boneName=attach_backpack}
    rotation {roll=15}
    translation {y=-5}
   }
   n3 = "Transform (Modify) Bone - Bone: attach_backpack"!(
    ComponentPose=@n2
    bAlphaBoolEnabled:bool=Get is_backpack_750() -> (is_backpack_750:bool)
    Translation="0,-1.5,0"
    Rotation="0,0,-9"
    Scale="1,1,1"
   ) -> (Pose) {
    alphaInputType=Bool
    rotationMode=BMM_Additive
    rotationSpace=BCS_BoneSpace
    translationMode=BMM_Additive
    translationSpace=BCS_ParentBoneSpace
    alphaBoolBlend {blendOption=Linear}
    boneToModify {boneName=attach_backpack}
    rotation {roll=-9}
    translation {y=-1.5}
   }
   n4 = "Transform (Modify) Bone - Bone: attach_backpack"(
    ComponentPose=@n3
    bAlphaBoolEnabled=NOT Boolean(A=Get Is FaceAcc Hood On() -> (Is FaceAcc Hood On:bool))
    Translation="0,-3.5,0"
    Rotation="0,0,-9"
    Scale="1,1,1"
   ) {
    alphaInputType=Bool
    rotationMode=BMM_Additive
    rotationSpace=BCS_BoneSpace
    translationMode=BMM_Additive
    translationSpace=BCS_ParentBoneSpace
    alphaBoolBlend {blendOption=Linear}
    boneToModify {boneName=attach_backpack}
    rotation {roll=-9}
    translation {y=-3.5}
   }
   n5 = RigidBody!(ComponentPose=@n4, bAlphaBoolEnabled:bool, OverrideWorldGravity:Vector=Get Gravity Override(), ExternalForce) -> (Pose) {
    alphaInputType=Bool
    bOverrideWorldGravity=true
    overridePhysicsAsset=&/F_MED_PawJasmineSword_Physics
    simulationSpace=BaseBoneSpace
    alphaBoolBlend {blendOption=Linear}
    baseBoneRef {boneName=pelvis}
    componentAppliedLinearAccClamp=<0 0 0>
    simSpaceSettings {
     maxAngularAcceleration=8
     maxAngularVelocity=8
     maxLinearAcceleration=400
     maxLinearVelocity=600
     worldAlpha=1
     externalLinearDragV=<0.1 0.1 0.1>
    }
   }
   Output Pose(Result=Component To Local(ComponentPose=@n5)) {layerGroup=None}
  ]}
 EventGraph {nodes [
   n1 = Event BlueprintUpdateAnimation()
   > n2 = Sequence!(execute) -> (then_0, then_1, then_2:Exec)
   > n3 = Switch on Int!(execute:Exec, Selection:int=Get FaceAcc_ID() -> (FaceAcc_ID:int=0)) -> (Default:Exec, 0:Exec, 1:Exec)
   Set Is FaceAcc Hood On(
    execute=@n3.0
    Is FaceAcc Hood On:bool=Get isHoodOn!(self:F MED Paw Jasmine Sword Face Acc Anim BP Object Reference=Get face_acc_animBp() -> (face_acc_animBp:F MED Paw Jasmine Sword Face Acc Anim BP Object Reference)) -> (isHoodOn:bool)
   ) -> (Output_Get:bool)
   Set Is FaceAcc Hood On(
    execute=@n3.1
    Is FaceAcc Hood On=Get isHoodOn(self:F MED Paw Jasmine Sword Mask Face Acc Anim BP Object Reference=Get face_acc_mask_animBp() -> (face_acc_mask_animBp:F MED Paw Jasmine Sword Mask Face Acc Anim BP Object Reference))
   )
   n4 = Gravity Override State Function DEPRECATED(
    execute=@n2.then_1
    CharacterPartAnimInstance=Self-Reference()
    GameTimeInSeconds=GetGameTimeInSeconds()
    GravityOverrideParams=Get Gravity Override Params()
    RBAN Params=Get RBAN Params()
   )
   > Set Gravity Override(Gravity Override=@n4)
   > Set WPO Globals Function(
    Material=Get Material Instance Dynamic() -> (Material Instance Dynamic:Material Instance Dynamic Object Reference)
    Overlay Material=Get Overlay Material Instance Dynamic() -> (Overlay Material Instance Dynamic:Material Instance Dynamic Object Reference)
    Skeletal Mesh=Get Skeletal Mesh() -> (Skeletal Mesh:Skeletal Mesh Component Object Reference)
    "Local Noise Transform  Location"="Get Local Noise Transform  Location"() -> ("Local Noise Transform  Location":Vector)
    Local Noise Offset=Get Local Noise Offset()
   )
   > Set WPO Regional Function!(
    execute
    Material:Material Instance Dynamic Object Reference=Get Material Instance Dynamic()
    Overlay Material:Material Instance Dynamic Object Reference=Get Overlay Material Instance Dynamic()
    Skeletal Mesh:Skeletal Mesh Component Object Reference=Get Skeletal Mesh()
    Velocity Transform:Name=pelvis
    Velocity Interp Speed:double=Get Velocity Interp Speed()
    Collision Transform:Name=pelvis
    Collision Transform Axis:EAxis Enum
    Invert Collision Transform Axis:bool=false
    Delta Time X:double=@n1.DeltaTimeX
    Previous Velocity Transform Location:"Vector (by ref)"=Get R1 Velocity Transform Location()
    Previous Velocity Transform Velocity:"Vector (by ref)"=Get R1 Velocity Transform Velocity()
    Region:int=1
   ) -> (then)
   > Set WPO Regional Function(
    Material=Get Material Instance Dynamic()
    Overlay Material=Get Overlay Material Instance Dynamic()
    Skeletal Mesh=Get Skeletal Mesh()
    Velocity Transform=spine_01
    Velocity Interp Speed=Get Velocity Interp Speed()
    Collision Transform=spine_05
    Invert Collision Transform Axis=true
    Delta Time X=@n1.DeltaTimeX
    Previous Velocity Transform Location=Get R2 Velocity Transform Location()
    Previous Velocity Transform Velocity=Get R2 Velocity Transform Velocity()
    Region=2
   )
   > n5 = Set isWPOEnabled(isWPOEnabled:bool=OR Boolean!(A=Get bIsSkydiveFloating(), B=Get bIsSkydiveDiveMode(), C:bool=Get bIsParachuteOpened()) -> (ReturnValue)) -> (Output_Get:bool)
   > n6 = Branch(Condition=@n5)
   > n7 = Set WPO Pho Wind Intensity(WPO Pho Wind Intensity=1)
   SetScalarParameterValue(
    execute=[@n9.then, @n7.then]
    self=[Get Overlay Material Instance Dynamic(), Get Material Instance Dynamic()]
    ParameterName=ClothSim_WindStrength
    Value=Get WPO Pho Wind Intensity()
   )
   n8 = vector * vector(A=GetRightVector!(self:Scene Component Object Reference=Get Skeletal Mesh()) -> (ReturnValue:Vector), B:Vector="-700,-700,-700") -> (ReturnValue=~)
   > Set!(
    execute
    Gravity Override Params_Skydive_GoingUp_28_#1_WindFrequency_28_#2:Vector="1.5,7,8"
    Gravity Override Params_Skydive_GoingUp_28_#1_WindAmplitude_31_#3:Vector="200,-500,0"
    Gravity Override Params_Skydive_GoingUp_28_#1_WindOffset_32_#4:Vector="300,500,0"
    Gravity Override Params_Skydive_GoingUp_28_#1_GravityOverride_15_#5:Vector=@n8
    Gravity Override Params_Skydive_GoingUp_28_#1_JointName_27_#6:Name=root
    Gravity Override Params_Skydive_FloatingDown_29_#7_WindFrequency_28_#2:Vector="1.5,7,8"
    Gravity Override Params_Skydive_FloatingDown_29_#7_WindAmplitude_31_#3:Vector="200,-500,0"
    Gravity Override Params_Skydive_FloatingDown_29_#7_WindOffset_32_#4:Vector="-400,-600,0"
    Gravity Override Params_Skydive_FloatingDown_29_#7_GravityOverride_15_#5:Vector=@n8
    Gravity Override Params_Skydive_FloatingDown_29_#7_JointName_27_#6:Name=root
    Gravity Override Params_Skydive_DivingDown_30_#8_WindFrequency_28_#2:Vector="1.5,12,10"
    Gravity Override Params_Skydive_DivingDown_30_#8_WindAmplitude_31_#3:Vector="50,300,0"
    Gravity Override Params_Skydive_DivingDown_30_#8_WindOffset_32_#4:Vector="0,0,0"
    Gravity Override Params_Skydive_DivingDown_30_#8_GravityOverride_15_#5:Vector="-1500,0,0"
    Gravity Override Params_Skydive_DivingDown_30_#8_JointName_27_#6:Name=pelvis
    Gravity Override Params_Skydive_Parachuting_31_#9:Gravity Override Setting Struct Structure=(WindFrequency_28_#2=(X=1.5,Y=7,Z=8),WindAmplitude_31_#3=(X=200,Y=-500,Z=400),WindOffset_32_#4=(X=0,Y=0,Z=0),GravityOverride_15_#5=(X=0,Y=-800,Z=-980),JointName_27_#6="root")
    Gravity Override Params_OnGround_Standing_32_#10:Gravity Override Setting Struct Structure=(WindFrequency_28_#2=(X=0,Y=0,Z=0),WindAmplitude_31_#3=(X=0,Y=0,Z=0),WindOffset_32_#4=(X=0,Y=0,Z=0),GravityOverride_15_#5=(X=0,Y=0,Z=-980),JointName_27_#6="root")
    Gravity Override Params_OnGround_Moving_33_#11:Gravity Override Setting Struct Structure=(WindFrequency_28_#2=(X=0,Y=0,Z=0),WindAmplitude_31_#3=(X=0,Y=0,Z=0),WindOffset_32_#4=(X=0,Y=0,Z=0),GravityOverride_15_#5=(X=0,Y=0,Z=-980),JointName_27_#6="root")
    Gravity Override Params_EmoteOrMelee_34_#12:Gravity Override Setting Struct Structure=(WindFrequency_28_#2=(X=0,Y=0,Z=0),WindAmplitude_31_#3=(X=0,Y=0,Z=0),WindOffset_32_#4=(X=0,Y=0,Z=0),GravityOverride_15_#5=(X=0,Y=0,Z=-980),JointName_27_#6="root")
    Gravity Override Params_Swimming_35_#13:Gravity Override Setting Struct Structure=(WindFrequency_28_#2=(X=0,Y=0,Z=0),WindAmplitude_31_#3=(X=0,Y=0,Z=0),WindOffset_32_#4=(X=0,Y=0,Z=0),GravityOverride_15_#5=(X=-800,Y=0,Z=0),JointName_27_#6="pelvis")
    Gravity Override Params_Front_End_36_#14:Gravity Override Setting Struct Structure=(WindFrequency_28_#2=(X=0,Y=0,Z=0),WindAmplitude_31_#3=(X=0,Y=0,Z=0),WindOffset_32_#4=(X=0,Y=0,Z=0),GravityOverride_15_#5=(X=0,Y=0,Z=-980),JointName_27_#6="root")
    Gravity Override Params_DBNO_40_#15:Gravity Override Setting Struct Structure=(WindFrequency_28_#2=(X=0,Y=0,Z=0),WindAmplitude_31_#3=(X=0,Y=0,Z=0),WindOffset_32_#4=(X=0,Y=0,Z=0),GravityOverride_15_#5=(X=-980,Y=400,Z=0),JointName_27_#6="pelvis")
    Gravity Override Params_DBNO_Carried_41_#16:Gravity Override Setting Struct Structure=(WindFrequency_28_#2=(X=0,Y=0,Z=0),WindAmplitude_31_#3=(X=0,Y=0,Z=0),WindOffset_32_#4=(X=0,Y=0,Z=0),GravityOverride_15_#5=(X=-980,Y=500,Z=-500),JointName_27_#6="pelvis")
    Gravity Override Params_OnGround_Falling_45_#17:Gravity Override Setting Struct Structure=(WindFrequency_28_#2=(X=0,Y=0,Z=0),WindAmplitude_31_#3=(X=0,Y=0,Z=0),WindOffset_32_#4=(X=0,Y=0,Z=0),GravityOverride_15_#5=(X=0,Y=0,Z=-980),JointName_27_#6="root")
    Gravity Override Params_Ziplining_46_#18:Gravity Override Setting Struct Structure=(WindFrequency_28_#2=(X=0,Y=0,Z=0),WindAmplitude_31_#3=(X=0,Y=0,Z=0),WindOffset_32_#4=(X=0,Y=0,Z=0),GravityOverride_15_#5=(X=0,Y=0,Z=-980),JointName_27_#6="root")
    Gravity Override Params_Skydive_Surfing_49_#19:Gravity Override Setting Struct Structure=(WindFrequency_28_#2=(X=1.5,Y=7,Z=8),WindAmplitude_31_#3=(X=200,Y=-500,Z=400),WindOffset_32_#4=(X=0,Y=0,Z=0),GravityOverride_15_#5=(X=500,Y=-300,Z=-980),JointName_27_#6="root")
    Gravity Override Params_Riding_Standing_54_#20:Gravity Override Setting Struct Structure=(WindFrequency_28_#2=(X=0,Y=0,Z=0),WindAmplitude_31_#3=(X=0,Y=0,Z=0),WindOffset_32_#4=(X=0,Y=0,Z=0),GravityOverride_15_#5=(X=0,Y=-350,Z=-980),JointName_27_#6="root")
    Gravity Override Params_Riding_Moving_55_#21:Gravity Override Setting Struct Structure=(WindFrequency_28_#2=(X=0,Y=0,Z=0),WindAmplitude_31_#3=(X=0,Y=0,Z=0),WindOffset_32_#4=(X=0,Y=0,Z=0),GravityOverride_15_#5=(X=0,Y=-450,Z=-450),JointName_27_#6="root")
    Gravity Override Params_Riding_Moving_Targeting_58_#22:Gravity Override Setting Struct Structure=(WindFrequency_28_#2=(X=0,Y=0,Z=0),WindAmplitude_31_#3=(X=0,Y=0,Z=0),WindOffset_32_#4=(X=0,Y=0,Z=0),GravityOverride_15_#5=(X=0,Y=-200,Z=-980),JointName_27_#6="root")
    Gravity Override Params_Tactical_Sprint_57_#23:Gravity Override Setting Struct Structure=(WindFrequency_28_#2=(X=0,Y=0,Z=0),WindAmplitude_31_#3=(X=0,Y=0,Z=0),WindOffset_32_#4=(X=0,Y=0,Z=0),GravityOverride_15_#5=(X=0,Y=-450,Z=-500),JointName_27_#6="root")
    Gravity Override Params_Rail_Grinding_61_#24:Gravity Override Setting Struct Structure=(WindFrequency_28_#2=(X=0,Y=0,Z=0),WindAmplitude_31_#3=(X=0,Y=0,Z=0),WindOffset_32_#4=(X=0,Y=0,Z=0),GravityOverride_15_#5=(X=0,Y=0,Z=-980),JointName_27_#6="root")
    Gravity Override Params_DBNO_Tumbling_67_#25:Gravity Override Setting Struct Structure=(WindFrequency_28_#2=(X=0,Y=0,Z=0),WindAmplitude_31_#3=(X=0,Y=0,Z=0),WindOffset_32_#4=(X=0,Y=0,Z=0),GravityOverride_15_#5=(X=-2000,Y=0,Z=0),JointName_27_#6="pelvis")
    Gravity Override Params_DBNO_RollFwd_68_#26:Gravity Override Setting Struct Structure=(WindFrequency_28_#2=(X=0,Y=0,Z=0),WindAmplitude_31_#3=(X=0,Y=0,Z=0),WindOffset_32_#4=(X=0,Y=0,Z=0),GravityOverride_15_#5=(X=-2000,Y=0,Z=0),JointName_27_#6="pelvis")
    Gravity Override Params_DBNO_RollBwd_69_#27:Gravity Override Setting Struct Structure=(WindFrequency_28_#2=(X=0,Y=0,Z=0),WindAmplitude_31_#3=(X=0,Y=0,Z=0),WindOffset_32_#4=(X=0,Y=0,Z=0),GravityOverride_15_#5=(X=-4000,Y=0,Z=0),JointName_27_#6="pelvis")
    Gravity Override Params_DBNO_Impulse_70_#28:Gravity Override Setting Struct Structure=(WindFrequency_28_#2=(X=0,Y=0,Z=0),WindAmplitude_31_#3=(X=0,Y=0,Z=0),WindOffset_32_#4=(X=0,Y=0,Z=0),GravityOverride_15_#5=(X=-2000,Y=0,Z=0),JointName_27_#6="pelvis")
   ) -> (then, Output_Get:Gravity Override Params Struct Structure)
   n9 = Set WPO Pho Wind Intensity(execute=@n6.else, WPO Pho Wind Intensity=0)
   n10 = GetDeltaTransformFromRefPose!(self=Get Skeletal Mesh(), BoneName=spine_01, BaseName=spine_05) -> (
    ReturnValue_Location:Vector
    ReturnValue_Rotation_Roll:float
    ReturnValue_Rotation_Pitch:float
    ReturnValue_Rotation_Yaw:float
    ReturnValue_Scale:Vector
   )
   Set back_transform(execute=@n2.then_2, back_transform:double=MapRangeClamped(Value=@n10.ReturnValue_Rotation_Yaw, InRangeA=5, InRangeB=30)) -> (Output_Get:double)
   > Set back_extra_transform(back_extra_transform:double=MapRangeClamped(Value=@n10.ReturnValue_Rotation_Yaw, InRangeA=25, InRangeB=70)) -> (Output_Get:double)
   n11 = Event OnCharacterCustomizationLoaded()
   n12 = Sequence(execute=[@n11.then, @n20.then])
   n13 = GetAnimInstance(self=Get Array Item with EFort Custom Part Type(Array=@n11.CustomMeshParts, PartType=Face))
   > n14 = Cast To F_MED_PawJasmineSword_FaceAcc_AnimBP!(execute, Object=@n13) -> (then, CastFailed, AsF MED Paw Jasmine Sword Face Acc Anim BP:F MED Paw Jasmine Sword Face Acc Anim BP Object Reference) {is=title K2Node_DynamicCast}
   > Set face_acc_animBp(face_acc_animBp:F MED Paw Jasmine Sword Face Acc Anim BP Object Reference=@n14) -> (Output_Get:F MED Paw Jasmine Sword Face Acc Anim BP Object Reference)
   > Set FaceAcc_ID(FaceAcc_ID:int=0) -> (Output_Get:int=0)
   n15 = Cast To F_MED_PawJasmineSwordMask_FaceAcc_AnimBP!(execute=@n14.CastFailed, Object=@n13) -> (then, CastFailed, AsF MED Paw Jasmine Sword Mask Face Acc Anim BP:F MED Paw Jasmine Sword Mask Face Acc Anim BP Object Reference) {is=title K2Node_DynamicCast}
   > Set face_acc_mask_animBp(face_acc_mask_animBp:F MED Paw Jasmine Sword Mask Face Acc Anim BP Object Reference=@n15) -> (Output_Get:F MED Paw Jasmine Sword Mask Face Acc Anim BP Object Reference)
   > Set FaceAcc_ID(FaceAcc_ID=1) -> (Output_Get=0)
   n16 = Set Skeletal Mesh(execute=@n12.then_1, Skeletal Mesh:Skeletal Mesh Component Object Reference=GetPartSkeletalMeshComponent()) -> (Output_Get:Skeletal Mesh Component Object Reference)
   > n17 = Get Dynamic Material Instance(SkeletalMeshComponent=@n16)
   > Set Material Instance Dynamic(Material Instance Dynamic:Material Instance Dynamic Object Reference=@n17) -> (Output_Get:Material Instance Dynamic Object Reference)
   > n18 = GetOrCreateDynamicOverlayMaterial(MeshComponent=@n16, MaterialName=MI_TechnoAwaken_Body)
   > Set Overlay Material Instance Dynamic(Overlay Material Instance Dynamic:Material Instance Dynamic Object Reference=@n18) -> (Output_Get:Material Instance Dynamic Object Reference)
   n19 = Get Array Item with EFort Custom Part Type(Array=@n11.CustomMeshParts, PartType=Backpack)
   Set is_backpack_750(
    execute=@n12.then_2
    is_backpack_750:bool="Equal (Object)"(
     A=Get SkeletalMesh!(self:Skinned Mesh Component Object Reference=@n19) -> (SkeletalMesh:Skeletal Mesh Object Reference)
     B=/BRCosmetics/Accessories/FORT_Backpacks/Backpack_M_MED_Raptor_Knight/Meshes/M_MED_Raptor_Knight_Pack
    )
   ) -> (Output_Get:bool)
   n20 = Event BlueprintInitializeAnimation()
  ]}
]
variables [
 back_extra_transform {type=number}
 back_transform {type=number}
 face_acc_animBp {type=F_MED_PawJasmineSword_FaceAcc_AnimBP_C, default=null}
 face_acc_mask_animBp {type=F_MED_PawJasmineSwordMask_FaceAcc_AnimBP_C, default=null}
 FaceAcc_ID {type=integer}
 forward_gravity_override {type=Vector}
 Gravity Override {type=Vector}
 Gravity Override Params {
  type=GravityOverrideParamsStruct
  default {skydive_FloatingDown {gravityOverride=<0 -700 0>, windAmplitude=<200 -500 0>, windOffset=<0 -600 0>}}
 }
 Is FaceAcc Hood On {type=boolean}
 is_backpack_750 {type=boolean}
 isWPOEnabled {type=boolean}
 Local Noise Offset {type=Vector}
 "Local Noise Transform  Location" {type=Vector}
 Material Instance Dynamic {type=MaterialInstanceDynamic}
 Overlay Material Instance Dynamic {type=MaterialInstanceDynamic}
 R1 Velocity Transform Location {type=Vector}
 R1 Velocity Transform Velocity {type=Vector}
 R2 Velocity Transform Location {type=Vector}
 R2 Velocity Transform Velocity {type=Vector}
 RBAN Params {
  type=RBANParamsStruct
  default {
   dBNO {simSpaceSettings {
     maxAngularAcceleration=1000
     maxAngularVelocity=3
     maxLinearAcceleration=300
     maxLinearVelocity=700
     velocityScaleZ=0
     externalLinearDragV=<0 0.5 0>
    }}
   dBNO_Carried {simSpaceSettings {
     maxAngularAcceleration=1000
     maxAngularVelocity=3
     maxLinearAcceleration=300
     maxLinearVelocity=700
     velocityScaleZ=0
     externalLinearDragV=<0 0.5 0>
    }}
   emoteOrMelee {simSpaceSettings {
     maxAngularAcceleration=1000
     maxAngularVelocity=3
     maxLinearAcceleration=300
     maxLinearVelocity=700
     velocityScaleZ=0
     externalLinearDragV=<0 0.5 0>
    }}
   front_End {simSpaceSettings {
     maxAngularAcceleration=1000
     maxAngularVelocity=3
     maxLinearAcceleration=300
     maxLinearVelocity=700
     velocityScaleZ=0
     externalLinearDragV=<0 0.5 0>
    }}
   onGround_Crouch_Moving {simSpaceSettings {
     maxAngularAcceleration=1000
     maxAngularVelocity=3
     maxLinearAcceleration=300
     maxLinearVelocity=700
     velocityScaleZ=0
     externalLinearDragV=<0 0.5 0>
    }}
   onGround_Falling {simSpaceSettings {
     maxAngularAcceleration=1000
     maxAngularVelocity=3
     maxLinearAcceleration=300
     maxLinearVelocity=700
     velocityScaleZ=0
     externalLinearDragV=<0 0.5 0>
    }}
   onGround_Moving {simSpaceSettings {
     maxAngularAcceleration=1000
     maxAngularVelocity=3
     maxLinearAcceleration=300
     maxLinearVelocity=700
     velocityScaleZ=0
     externalLinearDragV=<0 0.5 0>
    }}
   onGround_Standing {simSpaceSettings {
     maxAngularAcceleration=1000
     maxAngularVelocity=3
     maxLinearAcceleration=300
     maxLinearVelocity=700
     velocityScaleZ=0
     externalLinearDragV=<0 0.5 0>
    }}
   rail_Grinding {simSpaceSettings {
     maxAngularAcceleration=1000
     maxAngularVelocity=3
     maxLinearAcceleration=300
     maxLinearVelocity=700
     velocityScaleZ=0
     worldAlpha=0.1
     externalLinearDragV=<0 0.5 0>
    }}
   riding_Moving {simSpaceSettings {
     maxAngularAcceleration=1000
     maxAngularVelocity=3
     maxLinearAcceleration=300
     maxLinearVelocity=700
     velocityScaleZ=0
     externalLinearDragV=<0 0.5 0>
    }}
   riding_Moving_Targeting {simSpaceSettings {
     maxAngularAcceleration=1000
     maxAngularVelocity=3
     maxLinearAcceleration=300
     maxLinearVelocity=700
     velocityScaleZ=0
     externalLinearDragV=<0 0.5 0>
    }}
   riding_Standing {simSpaceSettings {
     maxAngularAcceleration=1000
     maxAngularVelocity=3
     maxLinearAcceleration=300
     maxLinearVelocity=700
     velocityScaleZ=0
     externalLinearDragV=<0 0.5 0>
    }}
   skydive_DivingDown {simSpaceSettings {
     maxAngularAcceleration=1000
     maxAngularVelocity=3
     maxLinearAcceleration=300
     maxLinearVelocity=700
     velocityScaleZ=0
     externalLinearDragV=<0 0.5 0>
    }}
   skydive_FloatingDown {
    componentLinearAccScale=<0 0.5 0>
    componentLinearVelScale=<0 0.5 0>
    simSpaceSettings {
     maxAngularAcceleration=1000
     maxAngularVelocity=3
     maxLinearAcceleration=300
     maxLinearVelocity=700
     velocityScaleZ=0
     externalLinearDragV=<0 0.5 0>
    }
   }
   skydive_GoingUp {simSpaceSettings {
     maxAngularAcceleration=1000
     maxAngularVelocity=3
     maxLinearAcceleration=300
     maxLinearVelocity=700
     velocityScaleZ=0
     externalLinearDragV=<0 0.5 0>
    }}
   skydive_Parachuting {
    componentLinearAccScale=<0 0.1 0.3>
    componentLinearVelScale=<0 0.1 0.3>
    simSpaceSettings {
     maxAngularAcceleration=1000
     maxAngularVelocity=3
     maxLinearAcceleration=300
     maxLinearVelocity=700
     velocityScaleZ=0
     externalLinearDragV=<0 0.5 0>
    }
   }
   skydive_Surfing {simSpaceSettings {
     maxAngularAcceleration=1000
     maxAngularVelocity=3
     maxLinearAcceleration=300
     maxLinearVelocity=700
     velocityScaleZ=0
     externalLinearDragV=<0 0.5 0>
    }}
   swimming {simSpaceSettings {
     maxAngularAcceleration=1000
     maxAngularVelocity=3
     maxLinearAcceleration=300
     maxLinearVelocity=700
     velocityScaleZ=0
     externalLinearDragV=<0 0.5 0>
    }}
   tactical_Sprint {simSpaceSettings {
     maxAngularAcceleration=1000
     maxAngularVelocity=3
     maxLinearAcceleration=300
     maxLinearVelocity=700
     velocityScaleZ=0
     externalLinearDragV=<0 0.5 0>
    }}
   ziplining {simSpaceSettings {
     maxAngularAcceleration=1000
     maxAngularVelocity=3
     maxLinearAcceleration=300
     maxLinearVelocity=700
     velocityScaleZ=0
     externalLinearDragV=<0 0.5 0>
    }}
  }
 }
 Skeletal Mesh {type=SkeletalMeshComponent}
 Velocity Interp Speed {type=number}
 WPO Pho Wind Intensity {type=number}
]
