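# Physics asset for the back hair of the F_MED_PawJasmineSword female character's face accessory part, defining capsule bodies and joints for physics-driven strands.
# Nine two-segment chains (right A–E; left A, B, D, E) each link to a master bone body, spelled MASTR or MSTR.
# All joints share equal swing limits and a tighter twist limit.
# Two larger, unconstrained zero-length capsules on hairHead_l/r act as spheres, apparently head colliders.
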PhysicsAsset /BRCosmetics/Characters/Player/Female/Medium/Bodies/F_MED_PawJasmineSword/Meshes/Parts/F_MED_PawJasmineSword_FaceAcc_Physics_HairBack
bodies {
 MASTR_hair_A_l_deform_1 [{shapeType=Capsule, length=0.505, radius=0.505, shapeName=MASTR_hair_A_l_deform_1_capsule, rotation=<rot 0 -90 0>}]
 MASTR_hair_A_r_deform_1 [{shapeType=Capsule, length=0.505, radius=0.505, shapeName=MASTR_hair_A_r_deform_1_capsule, rotation=<rot 0 -90 0>}]
 MASTR_hair_B_l_deform_1 [{shapeType=Capsule, length=0.505, radius=0.505, shapeName=MASTR_hair_B_l_deform_1_capsule, rotation=<rot 0 -90 0>}]
 MASTR_hair_B_r_deform_1 [{shapeType=Capsule, length=0.505, radius=0.505, shapeName=MASTR_hair_B_r_deform_1_capsule, rotation=<rot 0 -90 0>}]
 MASTR_hair_D_l_deform_1 [{shapeType=Capsule, length=0.505, radius=0.505, shapeName=MASTR_hair_D_l_deform_1_capsule, rotation=<rot 0 -90 0>}]
 MASTR_hair_E_l_deform_1 [{shapeType=Capsule, length=0.505, radius=0.505, shapeName=MASTR_hair_E_l_deform_1_capsule, rotation=<rot 0 -90 0>}]
 MSTR_hair_C_r_deform_1 [{shapeType=Capsule, length=0.505, radius=0.505, shapeName=MSTR_hair_C_r_deform_1_capsule, rotation=<rot 0 -90 0>}]
 MSTR_hair_D_r_deform_1 [{shapeType=Capsule, length=0.505, radius=0.505, shapeName=MSTR_hair_D_r_deform_1_capsule, rotation=<rot 0 -90 0>}]
 MSTR_hair_E_r_deform_1 [{shapeType=Capsule, length=0.505, radius=0.505, shapeName=MSTR_hair_E_r_deform_1_capsule, rotation=<rot 0 -90 0>}]
 hairHead_l [{shapeType=Capsule, length=0, radius=3.6775, shapeName=hairHead_l_capsule, center=<1.7227 -0.2734 0.1293>, rotation=<rot 15.2362 -90 8.0978>}]
 hairHead_r [{
   shapeType=Capsule
   length=0
   radius=3.7959
   shapeName=hairHead_r_capsule
   center=<1.9667 0.0977 -0.2962>
   rotation=<rot -20.5004 -90 -6.0349>
  }]
 hair_A_l_deform_1 [{shapeType=Capsule, length=0.3, radius=0.3, shapeName=hair_A_l_deform_1_capsule, center=<1.832 0 0>}]
 hair_A_l_deform_2 [{shapeType=Capsule, length=0.3, radius=0.3, shapeName=hair_A_l_deform_2_capsule, center=<1.1673 0 0>}]
 hair_A_r_deform_1 [{shapeType=Capsule, length=0.3, radius=0.3, shapeName=hair_A_r_deform_1_capsule, center=<1.1673 0 0>}]
 hair_A_r_deform_2 [{shapeType=Capsule, length=0.3, radius=0.3, shapeName=hair_A_r_deform_2_capsule, center=<1.1673 0 0>}]
 hair_B_l_deform_1 [{shapeType=Capsule, length=0.3, radius=0.3, shapeName=hair_B_l_deform_1_capsule, center=<1.1673 0 0>}]
 hair_B_l_deform_2 [{shapeType=Capsule, length=0.3, radius=0.3, shapeName=hair_B_l_deform_2_capsule, center=<1.1673 0 0>}]
 hair_B_r_deform_1 [{shapeType=Capsule, length=0.3, radius=0.3, shapeName=hair_B_r_deform_1_capsule, center=<1.1673 0 0>}]
 hair_B_r_deform_2 [{shapeType=Capsule, length=0.3, radius=0.3, shapeName=hair_B_r_deform_2_capsule, center=<1.1673 0 0>}]
 hair_C_r_deform_1 [{shapeType=Capsule, length=0.3, radius=0.3, shapeName=hair_C_r_deform_1_capsule, center=<1.2382 0 0>}]
 hair_C_r_deform_2 [{shapeType=Capsule, length=0.3, radius=0.3, shapeName=hair_C_r_deform_2_capsule, center=<1.1673 0 0>}]
 hair_D_l_deform_1 [{shapeType=Capsule, length=0.3, radius=0.3, shapeName=hair_D_l_deform_1_capsule, center=<1.1673 0 0>}]
 hair_D_l_deform_2 [{shapeType=Capsule, length=0.3, radius=0.3, shapeName=hair_D_l_deform_2_capsule, center=<1.1673 0 0>}]
 hair_D_r_deform_1 [{shapeType=Capsule, length=0.3, radius=0.3, shapeName=hair_D_r_deform_1_capsule, center=<1.1673 0 0>}]
 hair_D_r_deform_2 [{shapeType=Capsule, length=0.3, radius=0.3, shapeName=hair_D_r_deform_2_capsule, center=<1.1673 0 0>}]
 hair_E_l_deform_1 [{shapeType=Capsule, length=0.3, radius=0.3, shapeName=hair_E_l_deform_1_capsule, center=<1.1673 0 0>}]
 hair_E_l_deform_2 [{shapeType=Capsule, length=0.3, radius=0.3, shapeName=hair_E_l_deform_2_capsule, center=<1.1673 0 0>}]
 hair_E_r_deform_1 [{shapeType=Capsule, length=0.3, radius=0.3, shapeName=hair_E_r_deform_1_capsule, center=<1.1673 0 0>}]
 hair_E_r_deform_2 [{shapeType=Capsule, length=0.3, radius=0.3, shapeName=hair_E_r_deform_2_capsule, center=<1.1673 0 0>}]
}
constraints [
 {bone1Name=hair_A_r_deform_2, bone2Name=hair_A_r_deform_1, swing1LimitDegrees=30, swing2LimitDegrees=30, twistLimitDegrees=5}
 {bone1Name=hair_A_r_deform_1, bone2Name=MASTR_hair_A_r_deform_1, swing1LimitDegrees=30, swing2LimitDegrees=30, twistLimitDegrees=5}
 {bone1Name=hair_B_r_deform_2, bone2Name=hair_B_r_deform_1, swing1LimitDegrees=30, swing2LimitDegrees=30, twistLimitDegrees=5}
 {bone1Name=hair_B_r_deform_1, bone2Name=MASTR_hair_B_r_deform_1, swing1LimitDegrees=30, swing2LimitDegrees=30, twistLimitDegrees=5}
 {bone1Name=hair_C_r_deform_2, bone2Name=hair_C_r_deform_1, swing1LimitDegrees=30, swing2LimitDegrees=30, twistLimitDegrees=5}
 {bone1Name=hair_C_r_deform_1, bone2Name=MSTR_hair_C_r_deform_1, swing1LimitDegrees=30, swing2LimitDegrees=30, twistLimitDegrees=5}
 {bone1Name=hair_D_r_deform_2, bone2Name=hair_D_r_deform_1, swing1LimitDegrees=30, swing2LimitDegrees=30, twistLimitDegrees=5}
 {bone1Name=hair_D_r_deform_1, bone2Name=MSTR_hair_D_r_deform_1, swing1LimitDegrees=30, swing2LimitDegrees=30, twistLimitDegrees=5}
 {bone1Name=hair_E_r_deform_2, bone2Name=hair_E_r_deform_1, swing1LimitDegrees=30, swing2LimitDegrees=30, twistLimitDegrees=5}
 {bone1Name=hair_E_r_deform_1, bone2Name=MSTR_hair_E_r_deform_1, swing1LimitDegrees=30, swing2LimitDegrees=30, twistLimitDegrees=5}
 {bone1Name=hair_A_l_deform_2, bone2Name=hair_A_l_deform_1, swing1LimitDegrees=30, swing2LimitDegrees=30, twistLimitDegrees=5}
 {bone1Name=hair_A_l_deform_1, bone2Name=MASTR_hair_A_l_deform_1, swing1LimitDegrees=30, swing2LimitDegrees=30, twistLimitDegrees=5}
 {bone1Name=hair_B_l_deform_2, bone2Name=hair_B_l_deform_1, swing1LimitDegrees=30, swing2LimitDegrees=30, twistLimitDegrees=5}
 {bone1Name=hair_B_l_deform_1, bone2Name=MASTR_hair_B_l_deform_1, swing1LimitDegrees=30, swing2LimitDegrees=30, twistLimitDegrees=5}
 {bone1Name=hair_D_l_deform_2, bone2Name=hair_D_l_deform_1, swing1LimitDegrees=30, swing2LimitDegrees=30, twistLimitDegrees=5}
 {bone1Name=hair_D_l_deform_1, bone2Name=MASTR_hair_D_l_deform_1, swing1LimitDegrees=30, swing2LimitDegrees=30, twistLimitDegrees=5}
 {bone1Name=hair_E_l_deform_2, bone2Name=hair_E_l_deform_1, swing1LimitDegrees=30, swing2LimitDegrees=30, twistLimitDegrees=5}
 {bone1Name=hair_E_l_deform_1, bone2Name=MASTR_hair_E_l_deform_1, swing1LimitDegrees=30, swing2LimitDegrees=30, twistLimitDegrees=5}
]
