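# Physics asset for the front hair of the F_MED_PawJasmineSword cosmetic character's face accessory part.
# Left and right forehair strands are four-link capsule chains, each constrained up to its own small master root capsule.
# Swing limits loosen from root to tip while twist stays tight along each chain.
# A large head capsule has no constraints and provides a body for the hair to collide against.
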PhysicsAsset /BRCosmetics/Characters/Player/Female/Medium/Bodies/F_MED_PawJasmineSword/Meshes/Parts/F_MED_PawJasmineSword_FaceAcc_Physics_HairFront
bodies {
 MSTR_foreHair_l [{shapeType=Capsule, length=0.2, radius=0.5, shapeName=MSTR_foreHair_l_capsule}]
 MSTR_foreHair_r [{shapeType=Capsule, length=0.2, radius=0.5, shapeName=MSTR_foreHair_r_capsule}]
 foreHair_l_1 [{shapeType=Capsule, length=1, radius=1, shapeName=foreHair_l_1_capsule, center=<1.4263 0 0>, rotation=<rot -90 -0.0004 0>}]
 foreHair_l_2 [{
   shapeType=Capsule
   length=1.7129
   radius=1
   shapeName=foreHair_l_2_capsule
   center=<1.7848 0 0.2172>
   rotation=<rot -87.1876 179.9973 -179.9976>
  }]
 foreHair_l_3 [{shapeType=Capsule, length=1.7915, radius=0.9239, shapeName=foreHair_l_3_capsule, center=<1.8397 0.1074 0.0246>, rotation=<rot -90 0.0001 0>}]
 foreHair_l_4 [{
   shapeType=Capsule
   length=1.3076
   radius=0.6924
   shapeName=foreHair_l_4_capsule
   center=<1.4456 0.2976 -0.3305>
   rotation=<rot -78.7812 -175.7609 -175.7604>
  }]
 foreHair_r_1 [{shapeType=Capsule, length=1, radius=1, shapeName=foreHair_r_1_capsule, center=<1.4263 0 0>, rotation=<rot -90 0.0005 0>}]
 foreHair_r_2 [{shapeType=Capsule, length=1.7987, radius=0.9225, shapeName=foreHair_r_2_capsule, center=<1.8886 0 0>, rotation=<rot -90 0.0004 0>}]
 foreHair_r_3 [{shapeType=Capsule, length=1.5833, radius=1, shapeName=foreHair_r_3_capsule, center=<1.8126 0 0>, rotation=<rot -87.1876 -0.0003 0.0005>}]
 foreHair_r_4 [{
   shapeType=Capsule
   length=1.1681
   radius=0.8319
   shapeName=foreHair_r_4_capsule
   center=<1.5042 0.1769 0.1395>
   rotation=<rot -75.9388 -0.0003 11.2513>
  }]
 head [{shapeType=Capsule, length=6.0986, radius=7.9251, shapeName=head_capsule, center=<3.9862 0.4551 0>, rotation=<rot 0 -84.3753 -90>}]
}
constraints [
 {bone1Name=foreHair_l_4, bone2Name=foreHair_l_3, swing1LimitDegrees=30, swing2LimitDegrees=30, twistLimitDegrees=5}
 {bone1Name=foreHair_l_3, bone2Name=foreHair_l_2, swing1LimitDegrees=30, swing2LimitDegrees=30, twistLimitDegrees=5}
 {bone1Name=foreHair_l_2, bone2Name=foreHair_l_1, swing1LimitDegrees=30, swing2LimitDegrees=20, twistLimitDegrees=5}
 {bone1Name=foreHair_l_1, bone2Name=MSTR_foreHair_l, swing1LimitDegrees=30, swing2LimitDegrees=5, twistLimitDegrees=5}
 {bone1Name=foreHair_r_4, bone2Name=foreHair_r_3, swing1LimitDegrees=30, swing2LimitDegrees=30, twistLimitDegrees=5}
 {bone1Name=foreHair_r_3, bone2Name=foreHair_r_2, swing1LimitDegrees=30, swing2LimitDegrees=30, twistLimitDegrees=5}
 {bone1Name=foreHair_r_2, bone2Name=foreHair_r_1, swing1LimitDegrees=30, swing2LimitDegrees=20, twistLimitDegrees=5}
 {bone1Name=foreHair_r_1, bone2Name=MSTR_foreHair_r, swing1LimitDegrees=30, swing2LimitDegrees=5, twistLimitDegrees=5}
]
